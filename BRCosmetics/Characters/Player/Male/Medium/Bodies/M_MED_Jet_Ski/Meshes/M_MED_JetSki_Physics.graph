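# Physics asset for the male medium Jet Ski cosmetic body, covering its dangling gear: two grenades with pins, a pouch flap and four straps.
# Grenades, flap and straps each constrain to their own small Base anchor body, while each pin chains onto its grenade.
# Straps and pins lock twist and second swing for single-axis, hinge-like motion; grenades and flap leave those axes unlocked.
# Grenades use capsule shapes; pins, flap and straps use boxes.
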PhysicsAsset /BRCosmetics/Characters/Player/Male/Medium/Bodies/M_MED_Jet_Ski/Meshes/M_MED_JetSki_Physics
bodies {
 dyn_grenade_01 [{shapeType=Capsule, length=6.1697, radius=3.5544, center=<1.2604 0.2343 -0.1346>, rotation=<rot 0 -100.0248 -89.9146>}]
 dyn_grenade_01_Base [{shapeType=Capsule, length=0.505, radius=0.505, rotation=<rot 0 -90 0>}]
 dyn_grenade_02 [{shapeType=Capsule, length=6.1681, radius=3.658, center=<0.9462 -0.0095 -0.0602>, rotation=<rot 0 -95.2915 -89.2701>}]
 dyn_grenade_02_Base [{shapeType=Capsule, length=0.505, radius=0.505, rotation=<rot 0 -90 0>}]
 dyn_pin_01 [{shapeType=Box, extentX=0.7404, extentY=3.7232, extentZ=3.726, center=<1.6181 0 0>, rotation=<rot 0 89.0017 0>}]
 dyn_pin_02 [{shapeType=Box, extentX=0.7404, extentY=3.7232, extentZ=3.726, center=<1.6181 0 0>, rotation=<rot 0 89.0017 0>}]
 dyn_pouchFlap_01 [{shapeType=Box, extentX=1.1376, extentY=6.3499, extentZ=9.0886, center=<2.5793 -0.0376 0.115>, rotation=<rot 2.661 -90 -9.4325>}]
 dyn_pouchFlap_01_Base [{shapeType=Box, extentX=1.01, extentY=1.01, extentZ=1.01, rotation=<rot 0 -90 0>}]
 dyn_strap_01_Base_l [{shapeType=Box, extentX=1.01, extentY=1.01, extentZ=1.01, rotation=<rot 0 -90 0>}]
 dyn_strap_01_Base_r [{shapeType=Box, extentX=1.01, extentY=1.01, extentZ=1.01, rotation=<rot 0 -90 0>}]
 dyn_strap_01_l [{shapeType=Box, extentX=0.7256, extentY=3.8275, extentZ=4.2933, center=<1.1479 0.1148 -0.1356>, rotation=<rot 3.3221 -90 0>}]
 dyn_strap_01_r [{shapeType=Box, extentX=0.9314, extentY=3.8098, extentZ=4.2933, center=<1.1457 0.169 0.1528>, rotation=<rot -4.6462 -90 0>}]
 dyn_strap_02_Base_l [{shapeType=Box, extentX=1.01, extentY=1.01, extentZ=1.01, rotation=<rot 0 -90 0>}]
 dyn_strap_02_Base_r [{shapeType=Box, extentX=1.01, extentY=1.01, extentZ=1.01, rotation=<rot 0 -90 0>}]
 dyn_strap_02_l [{shapeType=Box, extentX=1.2945, extentY=4.2007, extentZ=4.5192, center=<0.8616 -0.0968 -0.1235>, rotation=<rot 9.3626 -90 0>}]
 dyn_strap_02_r [{shapeType=Box, extentX=1.176, extentY=3.4608, extentZ=3.4487, center=<1.2024 0.2106 -0.0049>, rotation=<rot -2.1951 -90 -3.5415>}]
}
constraints [
 {
  bone1Name=dyn_strap_01_r
  bone2Name=dyn_strap_01_Base_r
  swing1LimitDegrees=35
  swing2LimitDegrees=45
  swing2Motion=Locked
  twistLimitDegrees=45
  twistMotion=Locked
 }
 {
  bone1Name=dyn_strap_02_r
  bone2Name=dyn_strap_02_Base_r
  swing1LimitDegrees=35
  swing2LimitDegrees=45
  swing2Motion=Locked
  twistLimitDegrees=45
  twistMotion=Locked
 }
 {
  bone1Name=dyn_strap_02_l
  bone2Name=dyn_strap_02_Base_l
  swing1LimitDegrees=35
  swing2LimitDegrees=45
  swing2Motion=Locked
  twistLimitDegrees=45
  twistMotion=Locked
 }
 {
  bone1Name=dyn_strap_01_l
  bone2Name=dyn_strap_01_Base_l
  swing1LimitDegrees=35
  swing2LimitDegrees=45
  swing2Motion=Locked
  twistLimitDegrees=45
  twistMotion=Locked
 }
 {bone1Name=dyn_pouchFlap_01, bone2Name=dyn_pouchFlap_01_Base, swing1LimitDegrees=45, swing2LimitDegrees=16, twistLimitDegrees=16}
 {bone1Name=dyn_grenade_02, bone2Name=dyn_grenade_02_Base, swing1LimitDegrees=16, swing2LimitDegrees=16, twistLimitDegrees=16}
 {bone1Name=dyn_grenade_01, bone2Name=dyn_grenade_01_Base, swing1LimitDegrees=16, swing2LimitDegrees=16, twistLimitDegrees=16}
 {
  bone1Name=dyn_pin_01
  bone2Name=dyn_grenade_01
  swing1LimitDegrees=20
  swing2LimitDegrees=45
  swing2Motion=Locked
  twistLimitDegrees=45
  twistMotion=Locked
 }
 {
  bone1Name=dyn_pin_02
  bone2Name=dyn_grenade_02
  swing1LimitDegrees=20
  swing2LimitDegrees=45
  swing2Motion=Locked
  twistLimitDegrees=45
  twistMotion=Locked
 }
]
