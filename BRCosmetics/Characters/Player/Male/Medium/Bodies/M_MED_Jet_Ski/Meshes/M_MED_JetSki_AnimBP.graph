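# Animation Blueprint for the male medium Jet Ski cosmetic body, copying the attached parent mesh's pose and adding secondary motion to gear.
# Weighted multi-bone constraints place grenade, pouch flap, knife and strap base bones against the body.
# Right-thigh swing drives additive corrective transforms on both grenades; left-thigh swing is computed but unused.
# A final RigidBody node simulates physics in root-bone space, with interpolated alpha derived from targeting state.
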AnimBlueprint /BRCosmetics/Characters/Player/Male/Medium/Bodies/M_MED_Jet_Ski/Meshes/M_MED_JetSki_AnimBP
target_skeleton=&/M_MED_JetSki_Skeleton
graphs [
 AnimGraph {nodes [
   n1 = "Constraint - dyn_grenade_01_Base (PPPPPPP)"!(
    ComponentPose=Local To Component(LocalPose=Copy Pose From Mesh() {bUseAttachedParent=true})
    Alpha=1
    ConstraintWeights_0=0.0207
    ConstraintWeights_1:float=0.3603
    ConstraintWeights_2:float=0.0281
    ConstraintWeights_3:float=0.0086
    ConstraintWeights_4:float=0
    ConstraintWeights_5:float=0.5822
    ConstraintWeights_6:float=0
   ) -> (Pose) {
    alphaBoolBlend {blendOption=Linear}
    boneToModify {boneName=dyn_grenade_01_Base}
    constraintSetup [
     {targetBone {boneName=thigh_twist_01_r}}
     {targetBone {boneName=spine_01}}
     {targetBone {boneName=spine_02}}
     {targetBone {boneName=spine_03}}
     {targetBone {boneName=spine_04}}
     {targetBone {boneName=pelvis}}
     {targetBone {boneName=thigh_twist_01_l}}
    ]
    constraintWeights [1,1,1,1,1,1,1]
   }
   n2 = "Constraint - dyn_grenade_02_Base (PPPPP)"!(
    ComponentPose=@n1
    Alpha=1
    ConstraintWeights_0=0.3627
    ConstraintWeights_1:float=0.0077
    ConstraintWeights_2:float=0.006
    ConstraintWeights_3:float=0.0075
    ConstraintWeights_4:float=0.6161
   ) -> (Pose) {
    alphaBoolBlend {blendOption=Linear}
    boneToModify {boneName=dyn_grenade_02_Base}
    constraintSetup [
     {targetBone {boneName=spine_01}}
     {targetBone {boneName=spine_02}}
     {targetBone {boneName=spine_03}}
     {targetBone {boneName=thigh_twist_01_r}}
     {targetBone {boneName=pelvis}}
    ]
    constraintWeights [1,1,1,1,1]
   }
   n3 = "Constraint - dyn_pouchFlap_01_Base (PPP)"!(ComponentPose=@n2, Alpha=1, ConstraintWeights_0=0.1121, ConstraintWeights_1:float=0.0236, ConstraintWeights_2:float=0.8643) -> (Pose) {
    alphaBoolBlend {blendOption=Linear}
    boneToModify {boneName=dyn_pouchFlap_01_Base}
    constraintSetup [{targetBone {boneName=thigh_r}}, {targetBone {boneName=pelvis}}, {targetBone {boneName=thigh_twist_01_r}}]
    constraintWeights [1,1,1]
   }
   n4 = "Constraint - dyn_knife_01_Base (PPP)"!(ComponentPose=@n3, Alpha=1, ConstraintWeights_0=0.1338, ConstraintWeights_1:float=0.1301, ConstraintWeights_2:float=0.7361) -> (Pose) {
    alphaBoolBlend {blendOption=Linear}
    boneToModify {boneName=dyn_knife_01_Base}
    constraintSetup [{targetBone {boneName=calf_l}}, {targetBone {boneName=calf_twist_01_l}}, {targetBone {boneName=calf_twist_02_l}}]
    constraintWeights [1,1,1]
   }
   n5 = "Constraint - dyn_strap_01_Base_l (PPPPPPPPPPP)"!(
    ComponentPose=@n4
    Alpha=1
    ConstraintWeights_0=0.1684
    ConstraintWeights_1:float=0.0005
    ConstraintWeights_2:float=0.0697
    ConstraintWeights_3:float=0.1637
    ConstraintWeights_4:float=0.3581
    ConstraintWeights_5:float=0.2367
    ConstraintWeights_6:float=0.0012
    ConstraintWeights_7:float=0.0011
    ConstraintWeights_8:float=0
    ConstraintWeights_9:float=0.0006
    ConstraintWeights_10:float=0
   ) -> (Pose) {
    alphaBoolBlend {blendOption=Linear}
    boneToModify {boneName=dyn_strap_01_Base_l}
    constraintSetup [
     {targetBone {boneName=clavicle_l}}
     {targetBone {boneName=upperarm_twist_02_l}}
     {targetBone {boneName=spine_01}}
     {targetBone {boneName=spine_02}}
     {targetBone {boneName=spine_03}}
     {targetBone {boneName=spine_04}}
     {targetBone {boneName=spine_05}}
     {targetBone {boneName=upperarm_twist_01_l}}
     {targetBone {boneName=clavicle_r}}
     {targetBone {boneName=pelvis}}
     {targetBone {boneName=upperarm_l}}
    ]
    constraintWeights [1,1,1,1,1,1,1,1,1,1,1]
   }
   n6 = "Constraint - dyn_strap_02_Base_l (PPPPPPPPPP)"!(
    ComponentPose=@n5
    Alpha=1
    ConstraintWeights_0=0.0637
    ConstraintWeights_1:float=0
    ConstraintWeights_2:float=0.1394
    ConstraintWeights_3:float=0.3206
    ConstraintWeights_4:float=0.3506
    ConstraintWeights_5:float=0.0765
    ConstraintWeights_6:float=0
    ConstraintWeights_7:float=0
    ConstraintWeights_8:float=0.0492
    ConstraintWeights_9:float=0
   ) -> (Pose) {
    alphaBoolBlend {blendOption=Linear}
    boneToModify {boneName=dyn_strap_02_Base_l}
    constraintSetup [
     {targetBone {boneName=clavicle_l}}
     {targetBone {boneName=upperarm_twist_02_l}}
     {targetBone {boneName=spine_01}}
     {targetBone {boneName=spine_02}}
     {targetBone {boneName=spine_03}}
     {targetBone {boneName=spine_04}}
     {targetBone {boneName=spine_05}}
     {targetBone {boneName=upperarm_twist_01_l}}
     {targetBone {boneName=pelvis}}
     {targetBone {boneName=upperarm_l}}
    ]
    constraintWeights [1,1,1,1,1,1,1,1,1,1]
   }
   n7 = "Constraint - dyn_strap_01_Base_r (PPPPPPPPPPPPP)"!(
    ComponentPose=@n6
    Alpha=1
    ConstraintWeights_0=0
    ConstraintWeights_1:float=0
    ConstraintWeights_2:float=0.0011
    ConstraintWeights_3:float=0.0005
    ConstraintWeights_4:float=0.0697
    ConstraintWeights_5:float=0.1637
    ConstraintWeights_6:float=0.3581
    ConstraintWeights_7:float=0.2367
    ConstraintWeights_8:float=0.0012
    ConstraintWeights_9:float=0
    ConstraintWeights_10:float=0.1684
    ConstraintWeights_11:float=0.0006
    ConstraintWeights_12:float=0
   ) -> (Pose) {
    alphaBoolBlend {blendOption=Linear}
    boneToModify {boneName=dyn_strap_01_Base_r}
    constraintSetup [
     {targetBone {boneName=upperarm_r}}
     {targetBone {boneName=clavicle_l}}
     {targetBone {boneName=upperarm_twist_01_r}}
     {targetBone {boneName=upperarm_twist_02_r}}
     {targetBone {boneName=spine_01}}
     {targetBone {boneName=spine_02}}
     {targetBone {boneName=spine_03}}
     {targetBone {boneName=spine_04}}
     {targetBone {boneName=spine_05}}
     {targetBone {boneName=neck_02}}
     {targetBone {boneName=clavicle_r}}
     {targetBone {boneName=pelvis}}
     {targetBone {boneName=neck_01}}
    ]
    constraintWeights [1,1,1,1,1,1,1,1,1,1,1,1,1]
   }
   n8 = "Constraint - dyn_strap_02_Base_r (PPPPPPPPPPP)"!(
    ComponentPose=@n7
    Alpha=1
    ConstraintWeights_0=0
    ConstraintWeights_1:float=0
    ConstraintWeights_2:float=0
    ConstraintWeights_3:float=0
    ConstraintWeights_4:float=0.1394
    ConstraintWeights_5:float=0.3206
    ConstraintWeights_6:float=0.3506
    ConstraintWeights_7:float=0.0765
    ConstraintWeights_8:float=0
    ConstraintWeights_9:float=0.0637
    ConstraintWeights_10:float=0.0492
   ) -> (Pose) {
    alphaBoolBlend {blendOption=Linear}
    boneToModify {boneName=dyn_strap_02_Base_r}
    constraintSetup [
     {targetBone {boneName=upperarm_r}}
     {targetBone {boneName=clavicle_l}}
     {targetBone {boneName=upperarm_twist_01_r}}
     {targetBone {boneName=upperarm_twist_02_r}}
     {targetBone {boneName=spine_01}}
     {targetBone {boneName=spine_02}}
     {targetBone {boneName=spine_03}}
     {targetBone {boneName=spine_04}}
     {targetBone {boneName=spine_05}}
     {targetBone {boneName=clavicle_r}}
     {targetBone {boneName=pelvis}}
    ]
    constraintWeights [1,1,1,1,1,1,1,1,1,1,1]
   }
   n9 = Get legSwingDelta_r() -> (legSwingDelta_r:double)
   n10 = "Transform (Modify) Bone - Bone: dyn_grenade_01_Base"(ComponentPose=@n8, Alpha=@n9, Translation="-0.766,-0.368,-3.607", Rotation="-15.742,-3.332,16.388", Scale="1,1,1") {
    rotationMode=BMM_Additive
    rotationSpace=BCS_BoneSpace
    translationMode=BMM_Additive
    translationSpace=BCS_BoneSpace
    alphaBoolBlend {blendOption=Linear}
    boneToModify {boneName=dyn_grenade_01_Base}
    rotation=<rot -15.742 -3.332 16.388>
    translation=<-0.766 -0.368 -3.607>
   }
   n11 = "Transform (Modify) Bone - Bone: dyn_grenade_02_Base"(ComponentPose=@n10, Alpha=@n9, Translation="-0.000,0.286,-1.909", Rotation="0.000,-0.000,35.395", Scale="1,1,1") {
    rotationMode=BMM_Additive
    rotationSpace=BCS_BoneSpace
    translationMode=BMM_Additive
    translationSpace=BCS_BoneSpace
    alphaBoolBlend {blendOption=Linear}
    boneToModify {boneName=dyn_grenade_02_Base}
    rotation {roll=35.395}
    translation {y=0.286, z=-1.909}
   }
   n12 = RigidBody(ComponentPose=@n11, Alpha=Get targeting() -> (targeting:double)) {
    bClampLinearTranslationLimitToRefPose=true
    bOverrideWorldGravity=true
    overridePhysicsAsset=&/M_MED_JetSki_Physics
    simulationSpace=BaseBoneSpace
    alphaBoolBlend {blendOption=Linear}
    alphaScaleBiasClamp {bInterpResult=true, bMapRange=true, interpSpeedDecreasing=6, interpSpeedIncreasing=6, outRange {min=0.65}}
    baseBoneRef {boneName=root}
    componentAppliedLinearAccClamp=<3000 3000 3000>
    componentLinearAccScale=<0.5 0.5 0.7>
    componentLinearVelScale=<0.2 0.2 0.3>
    overrideWorldGravity {z=-980}
   }
   Output Pose(Result=Component To Local(ComponentPose=@n12)) {layerGroup=None}
  ]}
 EventGraph {nodes [
   Event BlueprintUpdateAnimation()
   n1 = GetOwnerSkeletalMeshComponent()
   > Set legSwingDelta_r(legSwingDelta_r:double=MapRangeClamped(Value=GetTwistAndSwingAngleOfDeltaRotationFromRefPose(self=@n1, BoneName=thigh_r).OutSwingAngle, InRangeA=174, InRangeB=72)) -> (Output_Get:double)
   > Set legSwingDelta_l(legSwingDelta_l:double=MapRangeClamped(Value=GetTwistAndSwingAngleOfDeltaRotationFromRefPose(self=@n1, BoneName=thigh_l).OutSwingAngle, InRangeA=24, InRangeB=101)) -> (Output_Get:double)
   n2 = Get bIsTargeting()
   > Set targeting(targeting:double=float - float(B="To Float (Boolean)"(InBool=@n2)) {is=title K2Node_CallFunction}) -> (Output_Get:double)
   TryGetPawnOwner()
  ]}
]
variables [
 legSwingDelta_l {type=number}
 legSwingDelta_r {type=number}
 targeting {type=number}
]
